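# AnimBlueprint for the F_MED_Graffiti_Remix female body, adding procedural secondary motion to clothing and accessory bones over a pose copied from the attached parent mesh.
# Weighted constraints and a Look At position drawstring, hoodie string, zipper and shoulder strap pivots.
# An interleaved AnimDynamics chain springs these plus necklace, mask, collar, straps, suspenders and shoe tongues; the mask has a left-thigh spherical limit.
# Finally, ThighLeft/RightRotationAlpha, not defined here, blend additive pelvis_l/r offsets.
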AnimBlueprint /BRCosmetics/Characters/Player/Female/Medium/Bodies/F_MED_Graffiti_Remix/Meshes/F_MED_Graffiti_Remix_AnimBP
target_skeleton=&/F_MED_Graffiti_Remix_Skeleton
graphs [
 AnimGraph {nodes [
   n1 = "Constraint - drawstring_pivot_zero (PPP)"!(
    ComponentPose=Local To Component(LocalPose=Copy Pose From Mesh() {bUseAttachedParent=true})
    Alpha=1
    ConstraintWeights_0=0.9642
    ConstraintWeights_1:float=0.0163
    ConstraintWeights_2:float=0.0196
   ) -> (Pose) {
    alphaBoolBlend {blendOption=Linear}
    boneToModify {boneName=drawstring_pivot_zero}
    constraintSetup [{targetBone {boneName=pelvis}}, {targetBone {boneName=thigh_twist_01_l}}, {targetBone {boneName=thigh_twist_01_r}}]
    constraintWeights [1,1,1]
   }
   n2 = "Constraint - drawstring_pivot_aim (PPPPP)"!(
    ComponentPose=@n1
    Alpha=1
    ConstraintWeights_0=0.0991
    ConstraintWeights_1:float=0.216
    ConstraintWeights_2:float=0.2341
    ConstraintWeights_3:float=0.3627
    ConstraintWeights_4:float=0.0881
   ) -> (Pose) {
    alphaBoolBlend {blendOption=Linear}
    boneToModify {boneName=drawstring_pivot_aim}
    constraintSetup [
     {targetBone {boneName=thigh_r}}
     {targetBone {boneName=pelvis}}
     {targetBone {boneName=thigh_twist_01_l}}
     {targetBone {boneName=thigh_twist_01_r}}
     {targetBone {boneName=thigh_l}}
    ]
    constraintWeights [1,1,1,1,1]
   }
   n3 = Look At - Bone: drawstring_pivot(ComponentPose=@n2, Alpha=1) {
    alphaBoolBlend {blendOption=Linear}
    boneToModify {boneName=drawstring_pivot}
    lookAtLocation {x=1}
    lookAtTarget {boneReference {boneName=drawstring_pivot_aim}}
    lookAt_Axis {axis=<1 0>}
    lookUp_Axis {axis {x=0, z=1}}
   }
   n4 = Anim Dynamics - Bone: drawstring_l(ComponentPose=@n3, Alpha=1, ExternalForce="0,0,0") {
    angularDampingOverride=1
    angularSpringConstant=1500
    bAngularSpring=true
    bGravityOverrideInSimSpace=true
    bOverrideAngularDamping=true
    gravityScale=0.35
    simulationSpace=RootRelative
    alphaBoolBlend {blendOption=Linear}
    boundBone {boneName=drawstring_l}
    physicsBodyDefinitions [{
      sphereCollisionRadius=0
      boundBone {boneName=drawstring_l}
      boxExtents=<10 20 20>
      constraintSetup {angularLimitsMax=<0 10 0>, angularLimitsMin=<0 -10 -40>, angularTarget=<0 0 0>}
      localJointOffset=<10 0 0>
     }]
   }
   n5 = Anim Dynamics - Bone: drawstring_r(ComponentPose=@n4, Alpha=1, ExternalForce="0,0,0") {
    angularDampingOverride=1
    angularSpringConstant=1500
    bAngularSpring=true
    bGravityOverrideInSimSpace=true
    bOverrideAngularDamping=true
    gravityScale=0.35
    simulationSpace=RootRelative
    alphaBoolBlend {blendOption=Linear}
    boundBone {boneName=drawstring_r}
    physicsBodyDefinitions [{
      sphereCollisionRadius=0
      boundBone {boneName=drawstring_r}
      boxExtents=<10 20 20>
      constraintSetup {angularLimitsMax=<0 10 0>, angularLimitsMin=<0 -10 -40>, angularTarget=<0 0 0>}
      localJointOffset=<10 0 0>
     }]
   }
   n6 = Anim Dynamics - Bone: dyn_necklace_0(ComponentPose=@n5, Alpha=1, ExternalForce="0,0,0") {
    angularDampingOverride=1
    angularSpringConstant=1500
    bAngularSpring=true
    bGravityOverrideInSimSpace=true
    bOverrideAngularDamping=true
    gravityScale=0.35
    simulationSpace=RootRelative
    alphaBoolBlend {blendOption=Linear}
    alphaScaleBiasClamp {scale=0.8}
    boundBone {boneName=dyn_necklace_0}
    physicsBodyDefinitions [{
      sphereCollisionRadius=0
      boundBone {boneName=dyn_necklace_0}
      boxExtents=<10 20 20>
      constraintSetup {angularLimitsMax=<0 10 0>, angularLimitsMin=<0 -10 -40>, angularTarget=<0 0 0>}
      localJointOffset=<10 0 0>
     }]
   }
   n7 = Anim Dynamics - Bone: dyn_necklace_1(ComponentPose=@n6, Alpha=1, ExternalForce="0,0,0") {
    angularDampingOverride=1
    angularSpringConstant=1500
    bAngularSpring=true
    bGravityOverrideInSimSpace=true
    bOverrideAngularDamping=true
    gravityScale=0.35
    simulationSpace=RootRelative
    alphaBoolBlend {blendOption=Linear}
    alphaScaleBiasClamp {scale=0.8}
    boundBone {boneName=dyn_necklace_1}
    physicsBodyDefinitions [{
      sphereCollisionRadius=0
      boundBone {boneName=dyn_necklace_1}
      boxExtents=<10 20 20>
      constraintSetup {angularLimitsMax=<0 0 0>, angularLimitsMin=<0 0 -40>, angularTarget=<0 0 0>}
      localJointOffset=<10 0 0>
     }]
   }
   n8 = "Constraint - hoodieString_0_l_zero (PP)"!(ComponentPose=@n7, Alpha=1, ConstraintWeights_0=0.89, ConstraintWeights_1:float=0.11) -> (Pose) {
    alphaBoolBlend {blendOption=Linear}
    boneToModify {boneName=hoodieString_0_l_zero}
    constraintSetup [{targetBone {boneName=spine_04}}, {targetBone {boneName=spine_05}}]
    constraintWeights [1,1]
   }
   n9 = "Constraint - hoodieString_0_r_zero (PP)"!(ComponentPose=@n8, Alpha=1, ConstraintWeights_0=0.8836, ConstraintWeights_1:float=0.1164) -> (Pose) {
    alphaBoolBlend {blendOption=Linear}
    boneToModify {boneName=hoodieString_0_r_zero}
    constraintSetup [{targetBone {boneName=spine_04}}, {targetBone {boneName=spine_05}}]
    constraintWeights [1,1]
   }
   n10 = Anim Dynamics - Bone: hoodieString_0_l(ComponentPose=@n9, Alpha=1, ExternalForce="0,0,0") {
    angularDampingOverride=1
    angularSpringConstant=1500
    bAngularSpring=true
    bGravityOverrideInSimSpace=true
    bOverrideAngularDamping=true
    gravityScale=0.35
    simulationSpace=RootRelative
    alphaBoolBlend {blendOption=Linear}
    boundBone {boneName=hoodieString_0_l}
    physicsBodyDefinitions [{
      sphereCollisionRadius=0
      boundBone {boneName=hoodieString_0_l}
      boxExtents=<10 20 20>
      constraintSetup {angularLimitsMax=<0 20 0>, angularLimitsMin=<0 -20 -40>, angularTarget=<0 0 0>}
      localJointOffset=<10 0 0>
     }]
   }
   n11 = Anim Dynamics - Bone: hoodieString_1_l(ComponentPose=@n10, Alpha=1, ExternalForce="0,0,0") {
    angularDampingOverride=1
    angularSpringConstant=1500
    bAngularSpring=true
    bGravityOverrideInSimSpace=true
    bOverrideAngularDamping=true
    gravityScale=0.35
    simulationSpace=RootRelative
    alphaBoolBlend {blendOption=Linear}
    boundBone {boneName=hoodieString_1_l}
    physicsBodyDefinitions [{
      sphereCollisionRadius=0
      boundBone {boneName=hoodieString_1_l}
      boxExtents=<10 20 20>
      constraintSetup {angularLimitsMax=<0 20 0>, angularLimitsMin=<0 -20 -40>, angularTarget=<0 0 0>}
      localJointOffset=<10 0 0>
     }]
   }
   n12 = Anim Dynamics - Bone: hoodieString_0_r(ComponentPose=@n11, Alpha=1, ExternalForce="0,0,0") {
    angularDampingOverride=1
    angularSpringConstant=1500
    bAngularSpring=true
    bGravityOverrideInSimSpace=true
    bOverrideAngularDamping=true
    gravityScale=0.35
    simulationSpace=RootRelative
    alphaBoolBlend {blendOption=Linear}
    boundBone {boneName=hoodieString_0_r}
    physicsBodyDefinitions [{
      sphereCollisionRadius=0
      boundBone {boneName=hoodieString_0_r}
      boxExtents=<10 20 20>
      constraintSetup {angularLimitsMax=<0 20 0>, angularLimitsMin=<0 -20 -40>, angularTarget=<0 0 0>}
      localJointOffset=<10 0 0>
     }]
   }
   n13 = Anim Dynamics - Bone: hoodieString_1_r(ComponentPose=@n12, Alpha=1, ExternalForce="0,0,0") {
    angularDampingOverride=1
    angularSpringConstant=1500
    bAngularSpring=true
    bGravityOverrideInSimSpace=true
    bOverrideAngularDamping=true
    gravityScale=0.35
    simulationSpace=RootRelative
    alphaBoolBlend {blendOption=Linear}
    boundBone {boneName=hoodieString_1_r}
    physicsBodyDefinitions [{
      sphereCollisionRadius=0
      boundBone {boneName=hoodieString_1_r}
      boxExtents=<10 20 20>
      constraintSetup {angularLimitsMax=<0 20 0>, angularLimitsMin=<0 -20 -40>, angularTarget=<0 0 0>}
      localJointOffset=<10 0 0>
     }]
   }
   n14 = "Constraint - hoodieZipper_zero (PP)"!(ComponentPose=@n13, Alpha=1, ConstraintWeights_0=0.9813, ConstraintWeights_1:float=0.0187) -> (Pose) {
    alphaBoolBlend {blendOption=Linear}
    boneToModify {boneName=hoodieZipper_zero}
    constraintSetup [{targetBone {boneName=spine_04}}, {targetBone {boneName=spine_05}}]
    constraintWeights [1,1]
   }
   n15 = Anim Dynamics - Bone: hoodieZipper(ComponentPose=@n14, Alpha=1, ExternalForce="0,0,0") {
    angularDampingOverride=1
    angularSpringConstant=1500
    bAngularSpring=true
    bGravityOverrideInSimSpace=true
    bOverrideAngularDamping=true
    gravityScale=0.35
    simulationSpace=RootRelative
    alphaBoolBlend {blendOption=Linear}
    boundBone {boneName=hoodieZipper}
    physicsBodyDefinitions [{
      sphereCollisionRadius=0
      boundBone {boneName=hoodieZipper}
      boxExtents=<10 20 20>
      constraintSetup {angularLimitsMax=<0 0 0>, angularLimitsMin=<0 0 -40>, angularTarget=<0 0 0>}
      localJointOffset=<10 0 0>
     }]
   }
   n16 = "Constraint - shoulderStrap_l_zero (PPP)"!(ComponentPose=@n15, Alpha=1, ConstraintWeights_0=0.6596, ConstraintWeights_1:float=0.1702, ConstraintWeights_2:float=0.1703) -> (Pose) {
    alphaBoolBlend {blendOption=Linear}
    boneToModify {boneName=shoulderStrap_l_zero}
    constraintSetup [{targetBone {boneName=shoulderStrap_l}}, {targetBone {boneName=upperarm_twist_01_l}}, {targetBone {boneName=clavicle_l}}]
    constraintWeights [1,1,1]
   }
   n17 = Anim Dynamics - Bone: shoulderStrap_l(ComponentPose=@n16, Alpha=1, ExternalForce="0,0,0") {
    angularDampingOverride=1
    angularSpringConstant=1500
    bAngularSpring=true
    bGravityOverrideInSimSpace=true
    bOverrideAngularDamping=true
    gravityScale=0.35
    simulationSpace=RootRelative
    alphaBoolBlend {blendOption=Linear}
    alphaScaleBiasClamp {scale=0.3}
    boundBone {boneName=shoulderStrap_l}
    componentLinearAccScale {z=0.3}
    componentLinearVelScale {z=0.3}
    physicsBodyDefinitions [{
      sphereCollisionRadius=0
      boundBone {boneName=shoulderStrap_l}
      boxExtents=<10 20 20>
      constraintSetup {angularLimitsMax=<0 0 0>, angularLimitsMin=<0 0 -40>, angularTarget=<0 0 0>}
      localJointOffset=<10 0 0>
     }]
   }
   n18 = "Constraint - shoulderStrap_r_zero (PPP)"!(ComponentPose=@n17, Alpha=1, ConstraintWeights_0=0.1832, ConstraintWeights_1:float=0.1827, ConstraintWeights_2:float=0.6341) -> (Pose) {
    alphaBoolBlend {blendOption=Linear}
    boneToModify {boneName=shoulderStrap_r_zero}
    constraintSetup [{targetBone {boneName=upperarm_twist_01_r}}, {targetBone {boneName=clavicle_r}}, {targetBone {boneName=shoulderStrap_r}}]
    constraintWeights [1,1,1]
   }
   n19 = Anim Dynamics - Bone: shoulderStrap_r(ComponentPose=@n18, Alpha=1, ExternalForce="0,0,0") {
    angularDampingOverride=1
    angularSpringConstant=1500
    bAngularSpring=true
    bGravityOverrideInSimSpace=true
    bOverrideAngularDamping=true
    gravityScale=0.35
    simulationSpace=RootRelative
    alphaBoolBlend {blendOption=Linear}
    alphaScaleBiasClamp {scale=0.3}
    boundBone {boneName=shoulderStrap_r}
    componentLinearAccScale {z=0.3}
    componentLinearVelScale {z=0.3}
    physicsBodyDefinitions [{
      sphereCollisionRadius=0
      boundBone {boneName=shoulderStrap_r}
      boxExtents=<10 20 20>
      constraintSetup {angularLimitsMax=<0 0 0>, angularLimitsMin=<0 0 -40>, angularTarget=<0 0 0>}
      localJointOffset=<10 0 0>
     }]
   }
   n20 = Anim Dynamics - Bone: mask_0(ComponentPose=@n19, Alpha=1, ExternalForce="0,0,0") {
    angularDampingOverride=1
    angularSpringConstant=1500
    bAngularSpring=true
    bGravityOverrideInSimSpace=true
    bOverrideAngularDamping=true
    bUseSphericalLimits=true
    gravityScale=0.35
    simulationSpace=RootRelative
    alphaBoolBlend {blendOption=Linear}
    boundBone {boneName=mask_0}
    physicsBodyDefinitions [{
      collisionType=CustomSphere
      sphereCollisionRadius=4.5
      boundBone {boneName=mask_0}
      boxExtents=<10 10 20>
      constraintSetup {angularLimitsMax=<0 0 0>, angularLimitsMin=<0 0 -40>, angularTarget=<0 0 0>}
      localJointOffset=<12 0 0>
     }]
    sphericalLimits [{limitRadius=13.5, drivingBone {boneName=thigh_l}, sphereLocalOffset=<-13.1311 0 3.7662>}]
   }
   n21 = Anim Dynamics - Bone: shoeTongue_l(ComponentPose=@n20, Alpha=1, ExternalForce="0,0,0") {
    angularDampingOverride=1
    angularSpringConstant=1500
    bAngularSpring=true
    bGravityOverrideInSimSpace=true
    bOverrideAngularDamping=true
    gravityScale=0.35
    simulationSpace=RootRelative
    alphaBoolBlend {blendOption=Linear}
    boundBone {boneName=shoeTongue_l}
    physicsBodyDefinitions [{
      sphereCollisionRadius=0
      boundBone {boneName=shoeTongue_l}
      boxExtents=<10 20 20>
      constraintSetup {angularLimitsMax=<0 0 0>, angularLimitsMin=<0 0 -40>, angularTarget=<0 0 0>}
      localJointOffset=<10 0 0>
     }]
   }
   n22 = Anim Dynamics - Bone: shoeTongue_r(ComponentPose=@n21, Alpha=1, ExternalForce="0,0,0") {
    angularDampingOverride=1
    angularSpringConstant=1500
    bAngularSpring=true
    bGravityOverrideInSimSpace=true
    bOverrideAngularDamping=true
    gravityScale=0.35
    simulationSpace=RootRelative
    alphaBoolBlend {blendOption=Linear}
    boundBone {boneName=shoeTongue_r}
    physicsBodyDefinitions [{
      sphereCollisionRadius=0
      boundBone {boneName=shoeTongue_r}
      boxExtents=<10 20 20>
      constraintSetup {angularLimitsMax=<0 0 0>, angularLimitsMin=<0 0 -40>, angularTarget=<0 0 0>}
      localJointOffset=<10 0 0>
     }]
   }
   n23 = Anim Dynamics - Bone: upperarm_strap_l(ComponentPose=@n22, Alpha=1, ExternalForce="0,0,0") {
    angularDampingOverride=1
    angularSpringConstant=1500
    bAngularSpring=true
    bGravityOverrideInSimSpace=true
    bOverrideAngularDamping=true
    gravityScale=0.35
    simulationSpace=RootRelative
    alphaBoolBlend {blendOption=Linear}
    boundBone {boneName=upperarm_strap_l}
    physicsBodyDefinitions [{
      sphereCollisionRadius=0
      boundBone {boneName=upperarm_strap_l}
      boxExtents=<10 20 20>
      constraintSetup {angularLimitsMax=<0 0 0>, angularLimitsMin=<0 0 -40>, angularTarget=<0 0 0>}
      localJointOffset=<10 0 0>
     }]
   }
   n24 = Anim Dynamics - Bone: collar_c_pivot(ComponentPose=@n23, Alpha=1, ExternalForce="0,0,0") {
    angularDampingOverride=0.9999
    angularSpringConstant=300
    bAngularSpring=true
    bGravityOverrideInSimSpace=true
    bOverrideAngularDamping=true
    gravityScale=0.1
    simulationSpace=RootRelative
    alphaBoolBlend {blendOption=Linear}
    alphaScaleBiasClamp {scale=0.2}
    boundBone {boneName=collar_c_pivot}
    componentAppliedLinearAccClamp=<0 0 0>
    physicsBodyDefinitions [{
      sphereCollisionRadius=0
      boundBone {boneName=collar_c_pivot}
      boxExtents=<5 15 15>
      constraintSetup {angularLimitsMax=<0 10 30>, angularLimitsMin=<0 -10 -60>, angularTarget=<1 0 0>}
      localJointOffset=<8 0 0>
     }]
   }
   n25 = Anim Dynamics - Bone: suspender_back_0_r(ComponentPose=@n24, Alpha=1, ExternalForce="0,0,0") {
    angularDampingOverride=1
    angularSpringConstant=1500
    bAngularSpring=true
    bGravityOverrideInSimSpace=true
    bOverrideAngularDamping=true
    gravityScale=0.35
    simulationSpace=RootRelative
    alphaBoolBlend {blendOption=Linear}
    boundBone {boneName=suspender_back_0_r}
    physicsBodyDefinitions [{
      sphereCollisionRadius=0
      boundBone {boneName=suspender_back_0_r}
      boxExtents=<10 20 20>
      constraintSetup {angularLimitsMax=<0 0 0>, angularLimitsMin=<0 0 -40>, angularTarget=<0 0 0>}
      localJointOffset=<10 0 0>
     }]
   }
   n26 = Anim Dynamics - Bone: suspender_front_0_r(ComponentPose=@n25, Alpha=1, ExternalForce="0,0,0") {
    angularDampingOverride=1
    angularSpringConstant=1500
    bAngularSpring=true
    bGravityOverrideInSimSpace=true
    bOverrideAngularDamping=true
    gravityScale=0.35
    simulationSpace=RootRelative
    alphaBoolBlend {blendOption=Linear}
    boundBone {boneName=suspender_front_0_r}
    physicsBodyDefinitions [{
      sphereCollisionRadius=0
      boundBone {boneName=suspender_front_0_r}
      boxExtents=<10 20 20>
      constraintSetup {angularLimitsMax=<0 0 0>, angularLimitsMin=<0 0 -40>, angularTarget=<0 0 0>}
      localJointOffset=<10 0 0>
     }]
   }
   n27 = Anim Dynamics - Bone: suspender_front_1_r(ComponentPose=@n26, Alpha=1, ExternalForce="0,0,0") {
    angularDampingOverride=1
    angularSpringConstant=1500
    bAngularSpring=true
    bGravityOverrideInSimSpace=true
    bOverrideAngularDamping=true
    gravityScale=0.35
    simulationSpace=RootRelative
    alphaBoolBlend {blendOption=Linear}
    boundBone {boneName=suspender_front_1_r}
    physicsBodyDefinitions [{
      sphereCollisionRadius=0
      boundBone {boneName=suspender_front_1_r}
      boxExtents=<10 20 20>
      constraintSetup {angularLimitsMax=<0 0 0>, angularLimitsMin=<0 0 -40>, angularTarget=<0 0 0>}
      localJointOffset=<10 0 0>
     }]
   }
   n28 = Anim Dynamics - Bone: suspender_back_1_r(ComponentPose=@n27, Alpha=1, ExternalForce="0,0,0") {
    angularDampingOverride=1
    angularSpringConstant=1500
    bAngularSpring=true
    bGravityOverrideInSimSpace=true
    bOverrideAngularDamping=true
    gravityScale=0.35
    simulationSpace=RootRelative
    alphaBoolBlend {blendOption=Linear}
    boundBone {boneName=suspender_back_1_r}
    physicsBodyDefinitions [{
      sphereCollisionRadius=0
      boundBone {boneName=suspender_back_1_r}
      boxExtents=<10 20 20>
      constraintSetup {angularLimitsMax=<0 0 0>, angularLimitsMin=<0 0 -40>, angularTarget=<0 0 0>}
      localJointOffset=<10 0 0>
     }]
   }
   n29 = "Transform (Modify) Bone - Bone: pelvis_l"(
    ComponentPose=@n28
    Alpha=Get ThighLeftRotationAlpha() -> (ThighLeftRotationAlpha:float)
    Translation="-9,0,0"
    Rotation="0,0,0"
    Scale="1,1,1"
   ) {
    translationMode=BMM_Additive
    translationSpace=BCS_ParentBoneSpace
    alphaBoolBlend {blendOption=Linear}
    boneToModify {boneName=pelvis_l}
    translation {x=-9}
   }
   n30 = "Transform (Modify) Bone - Bone: pelvis_r"(
    ComponentPose=@n29
    Alpha=Get ThighRightRotationAlpha() -> (ThighRightRotationAlpha:float)
    Translation="-20,0,0"
    Rotation="0,0,0"
    Scale="1,1,1"
   ) {
    translationMode=BMM_Additive
    translationSpace=BCS_ParentBoneSpace
    alphaBoolBlend {blendOption=Linear}
    boneToModify {boneName=pelvis_r}
    translation {x=-20}
   }
   Output Pose(Result=Component To Local(ComponentPose=@n30)) {layerGroup=None}
  ]}
 EventGraph {links [], nodes []}
]
variables []
